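# Translucent UI material for the VividRazor vault frame, drawing a progress bar with weakpoint markers and drill-attached variations.
# Mirrored texture lookups build frame and edge, while two notch layouts, one evenly spaced and one with per-notch offsets, supply weakpoint masks.
# Progress, remapped between start and end limits, positions the fill and a flame edge built from panned ember and flame textures, some on quantized time.
# Named reroutes connect the masks and colors, with IsDrillAttached? blending the final color.
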Material /VividRazorVaultGameplay/UI/Materials/M_UI_VividRazor_Frame
graph {
 properties {bAutomaticallySetUsageInEditor=false, blendMode=BLEND_Translucent, materialDomain=MD_UI, usedShadingModels=Default Lit}
 nodes [
  Constant() {desc="When progress value changes, it defines where change ends.\r\n0 - 1", r=0.9}
  Constant() {desc="When progress value changes, it defines where change starts.\r\n0 - 1", r=0.1}
  Constant() {desc="When progress value changes, it defines where change starts.\r\n0 - 1", r=0.1}
  Constant() {desc="When progress value changes, it defines where change starts.\r\n0 - 1", r=0.1}
  Constant() {desc="When progress value changes, it defines where change ends.\r\n0 - 1", r=0.9}
  n1 = NamedRerouteUsage()
  n2 = Max(A=Add(A=@n1, B=NamedRerouteUsage()), B=Multiply(A=@n1, B=NamedRerouteUsage()))
  LinearInterpolate(A=@n2, B=Add(A=@n2, B=NamedRerouteUsage()), Alpha=ScalarParameter() {parameterName=IsDrillAttached?})
  MaterialFunctionCall!(Value, Scale, Height) {materialFunction=/FortUILibrary/MaterialFunctions/Basics/MF_UI_PingPong}
  MaterialFunctionCall(Speed=ScalarParameter() {defaultValue=0.05, parameterName=DebugTimeSine}) {materialFunction=&DebugTimeSine}
  n3 = TextureCoordinate()
  n4 = ScalarParameter() {defaultValue=0.915, desc="When progress value changes, it defines where change ends.\r\n0 - 1", parameterName=ProgressEnd}
  n5 = Multiply(
   A=Subtract(A=ComponentMask(None=@n3) {r=true}, B=@n4) {desc=Offset by progress start}
   B=ScalarParameter() {defaultValue=5, parameterName=Weakpoint Count}
  ) {constB=4}
  n6 = Divide(B=Subtract(
    A=@n4
    B=ScalarParameter() {defaultValue=0.085, desc="When progress value changes, it defines where change starts.\r\n0 - 1", parameterName=ProgressStart}
   ))
  n7 = MaterialFunctionCall(
   Value=OneMinus(None=Multiply(A=@n5, B=@n6) {constB=4, desc=Shrink to progress zone})
   Height=Divide(B=ScalarParameter() {defaultValue=1.5, parameterName=PolyWidth})
  ) {materialFunction=&MF_UI_PingPong}
  n8 = Constant2Vector() {g=1.1, r=0.5}
  Max(A=@n7.Result, B=OneMinus(None=MaterialFunctionCall(Size=@n8) {materialFunction=&MF_UI_Box}.Fill)) {constB=0, desc=mask edges}
  n9 = TextureCoordinate()
  n10 = OneMinus(None=Multiply(A=ComponentMask(None=@n9) {r=true}) {constB=6})
  n11 = Abs(None=Add(A=@n10, B=ScalarParameter() {defaultValue=0.3, parameterName=Offset Notch 01}) {constB=0})
  n12 = Abs(None=Add(A=@n10, B=ScalarParameter() {defaultValue=1.6, parameterName=Offset Notch 02}) {constB=0})
  n13 = Add(A=@n10, B=ScalarParameter() {defaultValue=2.9, parameterName=Offset Notch 03}) {constB=0}
  n14 = ScalarParameter() {defaultValue=4.1, parameterName=Offset Notch 04}
  n15 = Min(A=Min(A=Min(A=@n11, B=@n12), B=Abs(None=@n13)), B=Abs(None=Add(A=@n10, B=@n14) {constB=0}))
  n16 = Add(A=ComponentMask(None=@n9) {g=true}) {constB=0.016}
  NamedRerouteDeclaration(None=MaterialFunctionCall(In=Add(A=@n15, B=@n16), Contrast=Constant() {r=30}) {materialFunction=&CheapContrast}.Result) {name=Hot Fixable: Independent Spacing, nodeColor=<rgba 0 0 0 1>}
  n17 = Constant2Vector() {g=1, r=1}
  n18 = MaterialFunctionCall!("Param (0-1)"=Time(), Tiles=Constant() {r=20}, Apply Offset) {materialFunction=/Game/Environments/Papaya/Materials/MaterialFunctions/MF_Papaya_Quantize_Scalar}
  n19 = TextureSample(UVs=Panner(
    Coordinate=MaterialFunctionCall(TextureSize=@n17) {materialFunction=&MF_UI_UVScaling}.Scaled to Fit
    Time=@n18.Result
    Speed=Constant2Vector() {g=-0.34, r=1.7298}
   ) {speedY=0.2}) {samplerType=SAMPLERTYPE_LinearGrayscale, texture=/VividRazorVaultGameplay/UI/Textures/T_UI_VividRazor_Flame}
  n20 = Constant2Vector() {g=0.974, r=1.6501}
  n21 = TextureSample(UVs=Panner(
    Coordinate=MaterialFunctionCall(TextureSize=@n20) {materialFunction=&MF_UI_UVScaling}.Scaled to Fit
    Time=@n18.Result
    Speed=Constant2Vector() {g=-0.34, r=1.7298}
   ) {speedY=0.2}) {samplerType=SAMPLERTYPE_LinearGrayscale, texture=&T_UI_VividRazor_Flame}
  NamedRerouteDeclaration(None=SmoothStep(Value=Subtract(A=@n19.RGB, B=@n21.RGB)) {constMax=0.9, constMin=0.6}) {name=Lil Ember BBs, nodeColor=<rgba 0 0 0 1>}
  n22 = TextureCoordinate()
  n23 = ScalarParameter() {
   desc="Property for animating progress value change (from ProgressStart to ProgressEnd)\r\n0 - 1"
   group=Progress
   parameterName=Progress
   sliderMax=1
   sortPriority=3
  }
  n24 = Add(
   A=Multiply(A=Saturate(None=@n22)) {constB=-1}
   B=LinearInterpolate(
    A=ScalarParameter() {defaultValue=0.085, desc="When progress value changes, it defines where change starts.\r\n0 - 1", parameterName=ProgressStart}
    B=ScalarParameter() {defaultValue=0.915, desc="When progress value changes, it defines where change ends.\r\n0 - 1", parameterName=ProgressEnd}
    Alpha=@n23
   )
  ) {constB=0.5}
  n25 = ComponentMask(None=@n24) {r=true}
  n26 = Add(A=@n25) {constB=0.018}
  n27 = ComponentMask(None=@n24) {r=true}
  n28 = Multiply(A=Saturate(None=SmoothStep(Value=@n26) {constMax=0.03}), B=Subtract(A=OneMinus(None=@n27)) {constB=0.82})
  n29 = Multiply(A=Saturate(None=Multiply(A=@n28) {constB=3}), B=NamedRerouteUsage())
  n30 = Multiply(A=@n29) {constB=5}
  NamedRerouteDeclaration(None=@n30) {name=FireShape, nodeColor=<rgba 0 0 0 1>}
  n31 = Multiply(A=NamedRerouteUsage(), B=NamedRerouteUsage())
  NamedRerouteDeclaration(None=@n31) {name=Ember BBs, nodeColor=<rgba 1 0 0.0177 1>}
  n32 = TextureCoordinate()
  n33 = Abs(None=Subtract(A=ComponentMask(None=@n32) {r=true}) {constB=0.5})
  n34 = Multiply(A=OneMinus(None=Multiply(A=@n33) {constB=2})) {constB=4}
  n35 = ComponentMask(None=@n32) {g=true}
  n36 = MaterialFunctionCall(UVs=AppendVector(A=Add(A=@n34) {constB=-0.45}, B=Add(A=@n35) {constB=0.045}), Texture Scale=Constant() {r=0.6}) {materialFunction=&ScaleUVsByCenter}
  NamedRerouteDeclaration(None=SmoothStep(Value=TextureSample(UVs=@n36.UVs) {texture=/VividRazorVaultGameplay/UI/Textures/T_UI_VividRazor}.A) {constMax=0.5, constMin=0.45}) {name=Edge, nodeColor=<rgba 0 0 0 1>}
  n37 = Saturate(None=SmoothStep(Max=Constant() {r=0.001}, Value=@n25) {constMax=0.01})
  NamedRerouteDeclaration(None=Multiply(A=Multiply(A=@n37), B=NamedRerouteUsage())) {name=ProgressGradient, nodeColor=<rgba 0 0 0 1>}
  n38 = Constant3Vector() {constant=<rgba 2 0.3216 0.034 1>}
  n39 = Add(A=@n38, B=@n38)
  n40 = NamedRerouteUsage()
  n41 = Add(A=@n39, B=@n39)
  n42 = LinearInterpolate(
   A=NamedRerouteUsage()
   B=LinearInterpolate(A=LinearInterpolate(A=@n38, B=@n39, Alpha=@n40), B=@n41, Alpha=NamedRerouteUsage())
   Alpha=NamedRerouteUsage()
  )
  n43 = Multiply(A=@n41) {constB=3}
  n44 = NamedRerouteUsage()
  NamedRerouteDeclaration(None=LinearInterpolate(A=NamedRerouteUsage(), B=LinearInterpolate(A=@n42, B=@n43, Alpha=@n44), Alpha=ScalarParameter() {parameterName=IsDrillAttached?})) {name=FinalColor, nodeColor=<rgba 0 0 0 1>}
  n45 = Constant2Vector() {g=0.152, r=0.8}
  n46 = Max(
   A=Multiply(A=NamedRerouteUsage(), B=MaterialFunctionCall(Size=@n45) {materialFunction=&MF_UI_Box}.Fill)
   B=NamedRerouteUsage()
  )
  NamedRerouteDeclaration(None=Multiply(A=Saturate(None=@n46)) {constB=0.9}) {name=ProgressBarWeakpoints, nodeColor=<rgba 0 0 0 1>}
  n47 = NamedRerouteUsage()
  n48 = NamedRerouteUsage()
  NamedRerouteDeclaration(None=LinearInterpolate(A=LinearInterpolate(B=@n47, Alpha=@n48), B=Constant3Vector() {constant=<rgba 0 0 0 1>}, Alpha=NamedRerouteUsage())) {name=Color: Base, nodeColor=<rgba 0 0 0 1>}
  n49 = MaterialFunctionCall!(UVs, Size=Constant2Vector() {r=0.6}, CornerRoundness, KeepAspectRatio, PixelSize) {materialFunction=/FortUILibrary/MaterialFunctions/Shapes/SDFs/MF_SDF_Box}
  NamedRerouteDeclaration(None=SmoothStep(Value=OneMinus(None=@n49.SDF)) {constMin=0.5}) {name=BoxMask, nodeColor=<rgba 0 0 0 1>}
  n50 = Saturate(None=@n7.Result)
  n51 = Add(
   A=ComponentMask(None=@n3) {g=true}
   B=LinearInterpolate(Alpha=ScalarParameter() {defaultValue=0.5, parameterName=PolyHeight}) {constA=0.08, constB=-0.08}
  ) {constB=0}
  NamedRerouteDeclaration(None=MaterialFunctionCall(In=Add(A=@n50, B=@n51), Contrast=Constant() {r=50}) {materialFunction=&CheapContrast}.Result) {name=Hot Fixable: Even Spacing, nodeColor=<rgba 0 0 0 1>}
  n52 = TextureCoordinate()
  n53 = ComponentMask(None=@n52) {r=true}
  n54 = Multiply(A=Abs(None=Subtract(A=@n53) {constB=0.5})) {constB=2}
  n55 = Add(A=Multiply(A=OneMinus(None=@n54)) {constB=5}) {constB=-0.003}
  n56 = Add(A=ComponentMask(None=@n52) {g=true}) {constB=0.042}
  n57 = SmoothStep(Value=TextureSample(UVs=AppendVector(A=@n55, B=@n56)) {texture=&T_UI_VividRazor}.R) {constMax=0.1, constMin=0.05}
  n58 = TextureCoordinate()
  n59 = ComponentMask(None=@n58) {r=true}
  n60 = Multiply(A=Abs(None=Subtract(A=@n59) {constB=0.5})) {constB=2}
  n61 = Add(A=Multiply(A=OneMinus(None=@n60)) {constB=4}) {constB=-0.304}
  n62 = Add(A=ComponentMask(None=@n58) {g=true}) {constB=0.25}
  n63 = SmoothStep(Value=TextureSample(UVs=AppendVector(A=@n61, B=@n62)) {texture=&T_UI_VividRazor}.G) {constMax=0.1, constMin=0.05}
  NamedRerouteDeclaration(None=Multiply(A=Add(A=@n57, B=@n63)) {constB=0.7}) {name=Frame, nodeColor=<rgba 0 0 0 1>}
  n64 = Panner(
   Coordinate=MaterialFunctionCall(TextureSize=Constant2Vector() {g=1.0786, r=2.1554}) {materialFunction=&MF_UI_UVScaling}.Scaled to Fit
   Speed=Constant2Vector() {g=0.104, r=2.1455}
  ) {speedY=0.5}
  n65 = Panner(
   Coordinate=MaterialFunctionCall(TextureSize=Constant2Vector() {g=0.974, r=1.6501}) {materialFunction=&MF_UI_UVScaling}.Scaled to Fit
   Speed=Constant2Vector() {g=-0.34, r=1.7298}
  ) {speedY=0.2}
  n66 = Multiply(B=Add(
    A=TextureSample(UVs=@n64) {samplerType=SAMPLERTYPE_LinearGrayscale, texture=/VividRazorVaultGameplay/UI/Textures/T_UI_VividRazor_Embers}.G
    B=TextureSample(UVs=@n65) {samplerType=SAMPLERTYPE_LinearGrayscale, texture=/VividRazorVaultGameplay/UI/Textures/T_UI_VividRazor_Embers_02}.G
   )) {constA=0.9}
  NamedRerouteDeclaration(None=Saturate(None=SmoothStep(Value=@n66) {constMax=0.7, constMin=0.05})) {name=FlameAlpha, nodeColor=<rgba 0 0 0 1>}
  n67 = LinearInterpolate(
   A=Constant3Vector() {constant=<rgba 0.9036 0.517 0 1>}
   B=Constant3Vector() {constant=<rgba 0.9479 0.7019 0 1>}
   Alpha=MaterialFunctionCall() {materialFunction=&LinearGradient}.UGradient
  )
  n68 = Multiply(A=@n67, B=@n67)
  n69 = TextureSample(UVs=Panner(Coordinate=TextureCoordinate() {uTiling=3, vTiling=2}, Speed=Constant2Vector() {r=0.75}) {speedY=0.2}) {samplerType=SAMPLERTYPE_LinearGrayscale, texture=/Game/Effects/Fort_Effects/Emotes/UltraLight/Texture/T_PlasmaCoils}
  NamedRerouteDeclaration(None=LinearInterpolate(A=@n67, B=Add(A=@n68, B=@n68), Alpha=Multiply(A=@n69.RGB) {constB=0.7})) {name=Color:Default, nodeColor=<rgba 0 0 0 1>}
  n70 = NamedRerouteUsage()
  n71 = Subtract(A=Saturate(None=@n30), B=@n70)
  n72 = Multiply(A=SmoothStep(Value=@n71) {constMax=1.5, constMin=0.75}) {constB=4}
  n73 = Max(
   A=Multiply(A=SmoothStep(Value=@n71) {constMax=0.8}) {constB=0.5}
   B=NamedRerouteDeclaration(None=Power(Base=@n72) {constExponent=1}) {name=InteriorFlame, nodeColor=<rgba 0 0 0 1>}
  )
  n74 = Multiply(A=@n29) {constB=2}
  n75 = Power(Base=@n74)
  n76 = Multiply(A=Subtract(A=Saturate(None=@n75), B=@n70)) {constB=0.2}
  n77 = Saturate(None=Power(Base=@n74))
  n78 = Multiply(A=NamedRerouteDeclaration(None=Add(A=@n76, B=@n77)) {name=FlameCenter, nodeColor=<rgba 0 0 0 1>}) {constB=2}
  NamedRerouteDeclaration(None=Max(A=Max(A=@n73, B=@n78), B=@n31)) {name=ProgressEdgeMask, nodeColor=<rgba 0 0 0 1>}
  NamedRerouteUsage()
  NamedRerouteUsage()
 ]
}
