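# Translucent UI material for the VividRazor weakpoint icon, switching between a target and a checkmark through IconSwitch.
# SmoothSteps on the icon texture's G and R channels cut each symbol's inner shape and dimmer backing, colored white or gold over black.
# Explode Color pushes the color toward pale cream and adjusts the target backing's opacity.
# Shatter drives a breakup mask by center-scaling a UV-remap texture and shifting its Step threshold.
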Material /VividRazorVaultGameplay/UI/Materials/M_UI_VividRazor_WeakpointIcon
graph {
 properties {bAutomaticallySetUsageInEditor=false, blendMode=BLEND_Translucent, materialDomain=MD_UI, usedShadingModels=Default Lit}
 nodes [
  n1 = Constant3Vector() {constant=<rgba 0 0 0 1>}
  n2 = Constant3Vector() {constant=<rgba 1 1 1 1>}
  n3 = NamedRerouteUsage()
  n4 = Constant3Vector() {constant=<rgba 0 0 0 1>}
  n5 = Constant3Vector() {constant=<rgba 1 0.76 0 1>}
  n6 = NamedRerouteUsage()
  LinearInterpolate(
   A=LinearInterpolate(
    A=LinearInterpolate(A=@n1, B=@n2, Alpha=@n3)
    B=Constant3Vector() {constant=<rgba 1 0.9812 0.7673 1>}
    Alpha=ScalarParameter() {parameterName=Explode Color}
   )
   B=LinearInterpolate(
    A=LinearInterpolate(A=@n4, B=@n5, Alpha=@n6)
    B=Constant3Vector() {constant=<rgba 1 0.9812 0.7673 1>}
    Alpha=ScalarParameter() {parameterName=Explode Color}
   )
   Alpha=ScalarParameter() {parameterName=IconSwitch}
  )
  MaterialFunctionCall(Speed=Constant() {r=0.5}) {materialFunction=&DebugTimeSine}
  n7 = TextureSample() {texture=/VividRazorVaultGameplay/UI/Textures/T_UI_VividRazor_WeakPointIcon}
  n8 = SmoothStep(Value=@n7.G) {constMax=0.4, constMin=0.3}
  n9 = Constant() {r=0.005}
  n10 = Multiply(A=SmoothStep(Max=@n9, Value=@n7.G) {constMax=0.004}, B=LinearInterpolate(Alpha=ScalarParameter() {parameterName=Explode Color}) {constA=0.6}) {constB=0.6}
  n11 = SmoothStep(Value=@n7.R) {constMax=0.4, constMin=0.3}
  n12 = Multiply(A=SmoothStep(Max=@n9, Value=@n7.R) {constMax=0.05}) {constB=0.6}
  n13 = ScalarParameter() {defaultValue=1, parameterName=Shatter}
  n14 = MaterialFunctionCall(Texture Scale=@n13) {materialFunction=&ScaleUVsByCenter}
  Multiply(
   A=LinearInterpolate(A=Add(A=@n8, B=@n10), B=Add(A=@n11, B=@n12), Alpha=ScalarParameter() {parameterName=IconSwitch})
   B=Step(
    Y=TextureSample(UVs=@n14.UVs) {samplerType=SAMPLERTYPE_Grayscale, texture=/VividRazorVaultGameplay/UI/Textures/T_UI_VividRazor_UVRemap}.R
    X=LinearInterpolate(Alpha=@n13) {constA=0.3}
   )
  )
  NamedRerouteDeclaration(None=@n11) {name=CheckMark Inner Alpha, nodeColor=<rgba 0 0 0 1>}
  NamedRerouteDeclaration(None=@n8) {name=Target Inner Alpha, nodeColor=<rgba 0 0 0 1>}
 ]
}
