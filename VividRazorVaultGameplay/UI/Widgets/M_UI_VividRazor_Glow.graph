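# Additive UI material that draws an orange-to-pale-yellow glow with a ring for the VividRazor Vault UI.
# A powered, inverted MF_UI_Circle SDF forms the circular glow, combined by Max with a ring cut from two exponential radial gradients.
# The Ring scalar drives both the centered UV scale and the cutout gradient's radius, changing the ring's size and shape.
# The combined mask feeds a three-color blend for the warm tint.
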Material /VividRazorVaultGameplay/UI/Widgets/M_UI_VividRazor_Glow
graph {
 properties {bAutomaticallySetUsageInEditor=false, blendMode=BLEND_Additive, materialDomain=MD_UI, usedShadingModels=Default Lit}
 nodes [
  n1 = Constant3Vector() {constant=<rgba 1 0.2661 0 1>}
  n2 = Constant3Vector() {constant=<rgba 1 0.6095 0.188 1>}
  LinearInterpolate(A=@n1, B=@n2)
  n3 = MaterialFunctionCall(Size=Constant(), OutlineThickness=Constant() {r=0.348}) {materialFunction=&MF_UI_Circle}
  n4 = Multiply(A=Power(Base=OneMinus(None=@n3.SDF)) {constExponent=7}) {constB=1.2}
  n5 = ScalarParameter() {defaultValue=1, parameterName=Ring}
  n6 = MaterialFunctionCall(Texture Scale=LinearInterpolate(Alpha=@n5) {constA=0.092, constB=3}) {materialFunction=&ScaleUVsByCenter}
  n7 = Multiply(A=MaterialFunctionCall(UVs=@n6.UVs, Radius=Constant() {r=0.164}, Density=Constant() {r=3.6726}) {materialFunction=&RadialGradientExponential}.RadialGradientExponential) {constB=2}
  n8 = Multiply(A=MaterialFunctionCall(UVs=@n6.UVs, Radius=LinearInterpolate(Alpha=@n5) {constA=0.1, constB=0.22}, Density=Constant() {r=1.381}) {materialFunction=&RadialGradientExponential}.RadialGradientExponential) {constB=7}
  n9 = Max(A=@n4, B=Min(A=Multiply(A=@n7, B=@n7) {constB=2}, B=OneMinus(None=@n8)))
  MaterialFunctionCall(A=@n1, B=@n2, C=Constant3Vector() {constant=<rgba 1 0.9682 0.7769 1>}, Alpha=@n9) {materialFunction=&3ColorBlend}
  Multiply(A=@n9) {constB=0.8}
  Saturate(None=@n4)
 ]
}
