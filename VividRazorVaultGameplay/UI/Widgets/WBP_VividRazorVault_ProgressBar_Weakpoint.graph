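# Weakpoint icon widget for the VividRazorVault progress bar, showing intact or destroyed state with an icon and glow inside a FortVisualAttachment.
# One animation drives material parameters, scale and glow opacity for the shatter-and-flash destruction effect.
# PlayWeakpointDestroyed plays it unless IsDestroyedState is set, without changing that flag.
# SetWeakpointDestroyed updates the flag on change or forced update, snapping to either state by playing from end time, forward or reversed.
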WidgetBlueprint /VividRazorVaultGameplay/UI/Widgets/WBP_VividRazorVault_ProgressBar_Weakpoint
animations [Anim_Icon_WeakpointDestroyed {
  binding_count=2
  length_seconds=0.65
  object=&self:Anim_Icon_WeakpointDestroyed
  sequence {
   bindings [
    Image_Glow {
     id=#1
     possessed_class=Image
     tracks [
      {
       type=MovieSceneFloatTrack
       display_name=Render Opacity
       sections [{
         type=MovieSceneFloatSection
         channels [None {type=MovieSceneScriptingFloatChannel, thinned_from=6, keys [[0,0],[7.155,1],[29.9,1],[38.636,0]]}]
        }]
      }
      {
       type=MovieScene2DTransformTrack
       sections [{
         type=MovieScene2DTransformSection
         channels [
          Translation.X {type=MovieSceneScriptingFloatChannel, const [0,1]}
          Translation.Y {type=MovieSceneScriptingFloatChannel, const [0,1]}
          Angle {type=MovieSceneScriptingFloatChannel, const [0,1]}
          Scale.X {type=MovieSceneScriptingFloatChannel, thinned_from=5, keys [[0,1],[8.586,6],[22.148,6],[38.759,1]]}
          Scale.Y {type=MovieSceneScriptingFloatChannel, keys [[0,1],[8.586,6],[22.148,6],[38.759,1]]}
          Shear.X {type=MovieSceneScriptingFloatChannel, const [0,1]}
          Shear.Y {type=MovieSceneScriptingFloatChannel, const [0,1]}
         ]
        }]
      }
      {
       type=MovieSceneWidgetMaterialTrack
       sections [{type=MovieSceneParameterSection, channels [Ring {type=MovieSceneScriptingFloatChannel, keys [[0,0],[18.111,1],[39,1]]}]}]
      }
     ]
    }
    Image_Icon {
     id=#2
     possessed_class=Image
     tracks [
      {
       type=MovieSceneWidgetMaterialTrack
       sections [{
         type=MovieSceneParameterSection
         channels [
          IconSwitch {type=MovieSceneScriptingFloatChannel, keys [[17.037,0],[19,1]]}
          Shatter {type=MovieSceneScriptingFloatChannel, keys [[0,1],[17.037,0.21],[37,0.8]]}
          Explode Color {type=MovieSceneScriptingFloatChannel, keys [[0,0],[8.519,1],[37,1],[39,0]]}
         ]
        }]
      }
      {
       type=MovieScene2DTransformTrack
       sections [{
         type=MovieScene2DTransformSection
         channels [
          Translation.X {type=MovieSceneScriptingFloatChannel, const [0,1]}
          Translation.Y {type=MovieSceneScriptingFloatChannel, const [0,1]}
          Angle {type=MovieSceneScriptingFloatChannel, const [0,1]}
          Scale.X {type=MovieSceneScriptingFloatChannel, keys [[0,1],[11.177,1.5],[13.412,4],[19.471,1.5],[24.589,1.5],[36.507,0.8],[38.942,1]]}
          Scale.Y {type=MovieSceneScriptingFloatChannel, keys [[0,1],[11.177,2.3],[13.412,4],[19.471,1.5],[24.589,1.5],[36.507,0.8],[38.942,1]]}
          Shear.X {type=MovieSceneScriptingFloatChannel, const [0,1]}
          Shear.Y {type=MovieSceneScriptingFloatChannel, const [0,1]}
         ]
        }]
      }
     ]
    }
   ]
   display_rate {numerator=60}
  }
 }]
graphs [
 EventGraph {links [], nodes []}
 PlayWeakpointDestroyed {nodes [
   PlayWeakpointDestroyed() {is=title K2Node_FunctionEntry}
   > n1 = Branch(Condition=Get IsDestroyedState() -> (IsDestroyedState:bool))
   PlayAnimation(execute=@n1.else, InAnimation=Get Anim_Icon_WeakpointDestroyed() -> (Anim_Icon_WeakpointDestroyed:Widget Animation Object Reference))
  ]}
 SetWeakpointDestroyed {nodes [
   n1 = SetWeakpointDestroyed!() -> (then, bForceUpdate:bool=~, IsDestroyed:bool=~) {is=title K2Node_FunctionEntry}
   > n2 = Branch(Condition=@n1)
   > n3 = Set IsDestroyedState(IsDestroyedState:bool=@n1.IsDestroyed) -> (Output_Get:bool)
   n4 = Delta Gate Wildcard!(execute:Exec=@n2.else, Variable:"Boolean (by ref)"=Get IsDestroyedState(), New Value:bool=@n1.IsDestroyed) -> (Success:Exec, Updated Variable:"Boolean (by ref)", Previous Variable:bool=~) {is=title K2Node_MacroInstance}
   n5 = Merge Execution with Data!(Option A:Exec=@n3.then, Option A Data:bool=@n3, Option B:Exec=@n4.Success, Option B Data:bool=@n4) -> (Then:Exec, Data:bool=~) {is=title K2Node_MacroInstance}
   n6 = Get Anim_Icon_WeakpointDestroyed()
   > PlayAnimation(
    InAnimation=@n6
    StartAtTime=GetEndTime(self=@n6)
    PlayMode=Select(Option 0:EUMGSequencePlayMode Enum=Reverse, Option 1:EUMGSequencePlayMode Enum, Index=@n5) -> (ReturnValue:EUMGSequencePlayMode Enum)
   )
  ]}
]
variables [IsDestroyedState {type=boolean}]
widgets {
 info {
  inheritedWidgetCount=0
  parentClass=&CommonUserWidget
  rootWidgetClass=&FortVisualAttachment
  widgetCount=4
 }
 widgets [
  FortVisualAttachment_0 {widgetClassPath=&FortVisualAttachment, widget=FortVisualAttachment_0, properties {contentAnchor=<0.5 0>}, slot=~}
  Overlay_Icon {widgetClassPath=&Overlay, parent=FortVisualAttachment_0, widget=Overlay_Icon, slot {class=SizeBoxSlot}}
  Image_Icon {
   widgetClassPath=&Image
   bIsVariable=true
   parent=Overlay_Icon
   widget=Image_Icon
   properties {
    visibility=SelfHitTestInvisible
    brush {
     bIsDynamicallyLoaded=false
     drawAs=Image
     imageType=FullColor
     mirroring=NoMirror
     resourceName=None
     resourceObject=/VividRazorVaultGameplay/UI/Materials/M_UI_VividRazor_WeakpointIcon
     tiling=NoTile
     imageSize=<45 45>
     margin=<ltrb 0 0 0 0>
     outlineSettings {
      bUseBrushTransparency=false
      roundingType=HalfHeightRadius
      width=0
      color {colorUseRule=UseColor_Specified, specifiedColor=<rgba 0 0 0 0>}
      cornerRadii=<0 0 0 0>
     }
     tintColor {colorUseRule=UseColor_Specified, specifiedColor=<rgba 1 1 1 1>}
     uVRegion {bIsValid=false, max=<0 0>, min=<0 0>}
    }
   }
   slot {class=OverlaySlot}
  }
  Image_Glow {
   widgetClassPath=&Image
   bIsVariable=true
   parent=Overlay_Icon
   widget=Image_Glow
   properties {
    renderOpacity=0
    visibility=SelfHitTestInvisible
    brush {
     bIsDynamicallyLoaded=false
     drawAs=Image
     imageType=FullColor
     mirroring=NoMirror
     resourceName=None
     resourceObject=&/M_UI_VividRazor_Glow
     tiling=NoTile
     imageSize=<32 32>
     margin=<ltrb 0 0 0 0>
     outlineSettings {
      bUseBrushTransparency=false
      roundingType=HalfHeightRadius
      width=0
      color {colorUseRule=UseColor_Specified, specifiedColor=<rgba 0 0 0 0>}
      cornerRadii=<0 0 0 0>
     }
     tintColor {colorUseRule=UseColor_Specified, specifiedColor=<rgba 1 1 1 1>}
     uVRegion {bIsValid=false, max=<0 0>, min=<0 0>}
    }
   }
   slot {class=OverlaySlot, properties {horizontalAlignment=HAlign_Center, verticalAlignment=VAlign_Center}}
  }
 ]
}
