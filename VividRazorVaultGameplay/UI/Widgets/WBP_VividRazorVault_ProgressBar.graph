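# HUD progress bar for Vivid Razor vault robberies, showing break-in progress and weakpoint markers.
# It resolves the vault manager through the playspace event router, polls progress every half second into a dynamic material, and spaces markers by weakpoint count.
# Router events drive damage, destruction and completion animations, with a timer holding the damage state.
# Participant gameplay tags toggle visibility; only active participants see real progress, otherwise status text appears.
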WidgetBlueprint /VividRazorVaultGameplay/UI/Widgets/WBP_VividRazorVault_ProgressBar
animations [
 Anim_WeakpointDamageBegin {
  binding_count=1
  length_seconds=0.0493
  object=&self:Anim_WeakpointDamageBegin
  sequence {bindings [Image_ProgressBar_Latest {
     id=#1
     possessed_class=Image
     tracks [{
       type=MovieSceneWidgetMaterialTrack
       sections [{
         type=MovieSceneParameterSection
         end_frame=0
         start_frame=0
         channels [IsDrillAttached? {type=MovieSceneScriptingFloatChannel, keys [[0,0],[0.986,1]]}]
        }]
      }]
    }]}
 }
 Anim_WeakpointDamageEnd {
  binding_count=1
  length_seconds=0.0952
  object=&self:Anim_WeakpointDamageEnd
  sequence {bindings [Image_ProgressBar_Latest {
     id=#2
     possessed_class=Image
     tracks [{
       type=MovieSceneWidgetMaterialTrack
       sections [{
         type=MovieSceneParameterSection
         end_frame=1
         start_frame=0
         channels [IsDrillAttached? {type=MovieSceneScriptingFloatChannel, keys [[0,1],[1.904,0]]}]
        }]
      }]
    }]}
 }
 Anim_WeakpointDestroyed {
  binding_count=1
  length_seconds=0.7
  object=&self:Anim_WeakpointDestroyed
  sequence {bindings [Image_ProgressBar_Latest {
     id=#3
     possessed_class=Image
     tracks [
      {
       type=MovieSceneWidgetMaterialTrack
       sections [{
         type=MovieSceneParameterSection
         end_frame=14
         start_frame=0
         channels [
          IsWeakPointDestroyed? {type=MovieSceneScriptingFloatChannel, const [0,3]}
          DeltaFill {type=MovieSceneScriptingFloatChannel, const [0,2]}
         ]
        }]
      }
      {
       type=MovieScene2DTransformTrack
       sections [{
         type=MovieScene2DTransformSection
         end_frame=10
         start_frame=0
         channels [
          Translation.X {type=MovieSceneScriptingFloatChannel, keys [[0,0],[1,-7.3158],[2,4.2502],[3,-8.1852],[4,4.5096],[10,0]]}
          Translation.Y {type=MovieSceneScriptingFloatChannel, keys [[0,0],[1,3.5512],[2,-5.5718],[3,3.2258],[4,-6.609],[10,0]]}
          Angle {type=MovieSceneScriptingFloatChannel, const [0,1]}
          Scale.X {type=MovieSceneScriptingFloatChannel, const [1,1]}
          Scale.Y {type=MovieSceneScriptingFloatChannel, const [1,1]}
          Shear.X {type=MovieSceneScriptingFloatChannel, const [0,1]}
          Shear.Y {type=MovieSceneScriptingFloatChannel, const [0,1]}
         ]
        }]
      }
     ]
    }]}
 }
 Anim_Complete {
  binding_count=2
  length_seconds=0.4963
  object=&self:Anim_Complete
  sequence {bindings [
    Overlay_ProgressBar {
     id=#4
     display_name=Overlay_0
     possessed_class=Overlay
     tracks [{
       type=MovieSceneFloatTrack
       display_name=Render Opacity
       sections [{type=MovieSceneFloatSection, end_frame=9, start_frame=0, channels [None {type=MovieSceneScriptingFloatChannel, keys [[0,1],[9.868,0]]}]}]
      }]
    }
    Border_BoostGlow {
     id=#5
     possessed_class=Border
     tracks [
      {
       type=MovieSceneFloatTrack
       display_name=Render Opacity
       sections [{
         type=MovieSceneFloatSection
         end_frame=9
         start_frame=0
         channels [None {type=MovieSceneScriptingFloatChannel, keys [[0,0],[5.9547,0.5],[9.9247,0]]}]
        }]
      }
      {
       type=MovieScene2DTransformTrack
       sections [{
         type=MovieScene2DTransformSection
         end_frame=5
         start_frame=0
         channels [
          Translation.X {type=MovieSceneScriptingFloatChannel, const [0,1]}
          Translation.Y {type=MovieSceneScriptingFloatChannel, const [0,1]}
          Angle {type=MovieSceneScriptingFloatChannel, const [0,1]}
          Scale.X {type=MovieSceneScriptingFloatChannel, keys [[0,0],[5.9547,1]]}
          Scale.Y {type=MovieSceneScriptingFloatChannel, const [0.35,1]}
          Shear.X {type=MovieSceneScriptingFloatChannel, const [0,1]}
          Shear.Y {type=MovieSceneScriptingFloatChannel, const [0,1]}
         ]
        }]
      }
     ]
    }
   ]}
 }
 Anim_EnterExit {
  binding_count=1
  length_seconds=0.2
  object=&self:Anim_EnterExit
  sequence {bindings [Overlay_ProgressBar {
     id=#6
     possessed_class=Overlay
     tracks [{
       type=MovieSceneFloatTrack
       display_name=Render Opacity
       sections [{type=MovieSceneFloatSection, end_frame=4, start_frame=0, channels [None {type=MovieSceneScriptingFloatChannel, keys [[0,0],[4,1]]}]}]
      }]
    }]}
 }
]
graphs [
 CleanupTimersAndReferences {nodes [
   CleanupTimersAndReferences() {is=title K2Node_FunctionEntry}
   > Clear and Invalidate Timer by Handle(Handle=Get ProgressUpdateTimerHandle() -> (ProgressUpdateTimerHandle:Timer Handle Structure))
   > Clear and Invalidate Timer by Handle(Handle=Get WeakpointDamageEventTimer() -> (WeakpointDamageEventTimer:Timer Handle Structure))
   > Set VaultEventRouter(VaultEventRouter:Gameplay Event Router Component Object Reference) -> (Output_Get:Gameplay Event Router Component Object Reference)
   > Set CurrentManager(CurrentManager:BGA Vivid Razor Vault Manager Object Reference) -> (Output_Get:BGA Vivid Razor Vault Manager Object Reference)
   > ClearRouterBindings()
  ]}
 ClearRouterBindings {nodes [
   ClearRouterBindings() {is=title K2Node_FunctionEntry}
   > StopListeningToEndEvent()
   > StopListeningToWeakpoinDamaged()
   > StopListeningToWeakpointDestroyed()
  ]}
 EventGraph {nodes [
   SetWeakpointCountFromVault() {is=title K2Node_CustomEvent}
   > SetWeakpointCount!(execute, self, Count:int=GetWeakpointCount!(self) -> (Count:int=0)) -> (then)
   StopListeningToEndEvent() {is=title K2Node_CustomEvent}
   > n1 = Get!(execute:Exec) -> (then:Exec, else:Exec, VaultOpenAsyncListener:Async Action Start Listening To Stateful Event Object Reference)
   > StopListeningToStatefulEvent(self:Async Action Start Listening To Stateful Event Object Reference=@n1)
   > Set VaultOpenAsyncListener(VaultOpenAsyncListener:Async Action Start Listening To Stateful Event Object Reference) -> (Output_Get:Async Action Start Listening To Stateful Event Object Reference)
   ListenToEndEvent() {is=title K2Node_CustomEvent}
   > n2 = StartListeningToStatefulEvent!(
    execute:Exec
    Target:Gameplay Event Router Component Object Reference=Get VaultEventRouter() -> (VaultEventRouter:Gameplay Event Router Component Object Reference)
    EventType:Script Struct Object Reference=/VividRazorCoreGameplay/Gameplay/Structs/Struct_VividRazorCore_Vault_Opened_Payload
    Context:Object Reference
    EventBubblingRule:EEventBubblingRule Enum
    None:%=Self-Reference()
   ) -> (
    then:Exec
    AsyncTaskProxy:Async Action Start Listening To Stateful Event Object Reference
    OnEventReceived:Exec
    OnSavedEventStateExists:Exec
    OnEventStateCleared:Exec
    Payload:Struct Vivid Razor Core Vault Opened Payload Structure
   ) {is=title K2Node_StartListeningToStatefulEventAsyncAction}
   > Set VaultOpenAsyncListener(VaultOpenAsyncListener=@n2)
   n3 = HasMatchingGameplayTag!(
    execute:Exec=@n2.OnEventReceived
    self:Object Reference=Get Controlled Pawn!(self:Controller Object Reference=GetOwningPlayer()) -> (ReturnValue:Pawn Object Reference) {is=title K2Node_CallFunction}
    TagToCheck:Gameplay Tag Structure=Get ActiveVaultParticipantTag() -> (ActiveVaultParticipantTag:Gameplay Tag Structure)
   ) -> (then:Exec, ReturnValue:bool) {is=title K2Node_Message}
   > n4 = Branch(Condition=@n3)
   > InitializeFromRouter!(execute, self, GameplayEventRouter:Gameplay Event Router Component Object Reference) -> (then)
   "Play Animation with Finished event (legacy)"!(
    execute:Exec=@n4.else
    Widget:User Widget Object Reference=Self-Reference()
    InAnimation:Widget Animation Object Reference=Get Anim_Complete() -> (Anim_Complete:Widget Animation Object Reference)
    StartAtTime:float=0
    NumLoopsToPlay:int=1
    PlayMode:EUMGSequencePlayMode Enum
    PlaybackSpeed:float=1
   ) -> (then:Exec, Result:UMGSequence Player Object Reference, Finished:Exec) {is=title K2Node_PlayAnimation}
   StopListeningToWeakpoinDamaged() {is=title K2Node_CustomEvent}
   > n5 = Get!(execute) -> (then, else, WeakpointDamagedAsyncListener:Async Action Start Listening To Event Object Reference)
   > StopListeningToEvent(self:Async Action Start Listening To Event Object Reference=@n5)
   > Set WeakpointDamagedAsyncListener(WeakpointDamagedAsyncListener:Async Action Start Listening To Event Object Reference) -> (Output_Get:Async Action Start Listening To Event Object Reference)
   StopListeningToWeakpointDestroyed() {is=title K2Node_CustomEvent}
   > n6 = Get!(execute) -> (then, else, WeakpointDestroyedAsyncListener:Async Action Start Listening To Event Object Reference)
   > StopListeningToEvent(self=@n6)
   > Set WeakpointDestroyedAsyncListener(WeakpointDestroyedAsyncListener:Async Action Start Listening To Event Object Reference) -> (Output_Get:Async Action Start Listening To Event Object Reference)
   ListenToWeakpointDestroyed() {is=title K2Node_CustomEvent}
   > n7 = StartListeningToEvent!(
    execute:Exec
    Target:Gameplay Event Router Component Object Reference=Get VaultEventRouter()
    EventType:Script Struct Object Reference=/VividRazorCoreGameplay/Gameplay/Structs/Struct_VividRazorCore_Vault_Weakpoint_Destroyed_Payload
    Context:Object Reference
    EventBubblingRule:EEventBubblingRule Enum
    None:%=Self-Reference()
   ) -> (
    then:Exec
    AsyncTaskProxy:Async Action Start Listening To Event Object Reference
    OnEventReceived:Exec
    Payload:Struct Vivid Razor Core Vault Weakpoint Destroyed Payload Structure
   )
   > Set WeakpointDestroyedAsyncListener(WeakpointDestroyedAsyncListener=@n7)
   OnWeakpointDestroyed(execute=@n7.OnEventReceived)
   ListenToAlertStatusReplicated() {is=title K2Node_CustomEvent}
   > n8 = Is Valid(InputObject=Get AlertStatusReplicatedAsyncListener() -> (AlertStatusReplicatedAsyncListener:Async Action Start Listening To Event Object Reference))
   n9 = StartListeningToEvent(
    execute=@n8.Is Not Valid
    Target=GetGlobalEventRouter()
    EventType=/VividRazorCoreGameplay/Gameplay/Structs/Struct_VividRazorCore_Vault_AlertStatusReplicated_Payload
    None=Self-Reference()
   ) -> (Payload:Struct Vivid Razor Core Vault Alert Status Replicated Payload Structure)
   > Set AlertStatusReplicatedAsyncListener(AlertStatusReplicatedAsyncListener:Async Action Start Listening To Event Object Reference=@n9) -> (Output_Get:Async Action Start Listening To Event Object Reference)
   InitializeFromRouter(execute=@n9.OnEventReceived)
   > PlayEnterExit!(execute, self, IsEntering:bool=true) -> (then)
   n10 = ListenForParticipant!() -> (OutputDelegate, then, Pawn:Pawn Object Reference) {is=title K2Node_CustomEvent}
   > n11 = Cast To FortPawn!(execute, Object=@n10.Pawn) -> (then, CastFailed, AsFort Pawn:Fort Pawn Object Reference) {is=title K2Node_DynamicCast}
   n12 = Get PlayerState!(self:Pawn Object Reference=@n11) -> (PlayerState:Player State Object Reference)
   > n13 = WaitGameplayTagAddToActor!(
    execute:Exec
    TargetActor:Actor Object Reference=@n12
    Tag:Gameplay Tag Structure=Get VaultParticipantTag() -> (VaultParticipantTag:Gameplay Tag Structure)
    OnlyTriggerOnce:bool=false
   ) -> (then:Exec, AsyncTaskProxy:Ability Async Wait Gameplay Tag Added Object Reference, Added:Exec) {is=title K2Node_AsyncAction}
   > Set ParticipantTagAddListener(ParticipantTagAddListener:Ability Async Wait Gameplay Tag Added Object Reference=@n13) -> (Output_Get:Ability Async Wait Gameplay Tag Added Object Reference)
   > n14 = WaitGameplayTagRemoveFromActor!(execute:Exec, TargetActor:Actor Object Reference=@n12, Tag:Gameplay Tag Structure=Get VaultParticipantTag(), OnlyTriggerOnce:bool=false) -> (then:Exec, AsyncTaskProxy:Ability Async Wait Gameplay Tag Removed Object Reference, Removed:Exec) {is=title K2Node_AsyncAction}
   > Set ParticipantTagRemoveListener(ParticipantTagRemoveListener:Ability Async Wait Gameplay Tag Removed Object Reference=@n14) -> (Output_Get:Ability Async Wait Gameplay Tag Removed Object Reference)
   > n15 = WaitGameplayTagAddToActor(TargetActor=@n12, Tag=Get ActiveVaultParticipantTag(), OnlyTriggerOnce=false)
   > Set ActiveParticipantTagAddListener(ActiveParticipantTagAddListener:Ability Async Wait Gameplay Tag Added Object Reference=@n15) -> (Output_Get:Ability Async Wait Gameplay Tag Added Object Reference)
   > n16 = WaitGameplayTagRemoveFromActor(TargetActor=@n12, Tag=Get ActiveVaultParticipantTag(), OnlyTriggerOnce=false)
   > Set ActiveParticipantTagRemoveListener(ActiveParticipantTagRemoveListener:Ability Async Wait Gameplay Tag Removed Object Reference=@n16) -> (Output_Get:Ability Async Wait Gameplay Tag Removed Object Reference)
   > SetIsActiveParticipant!(
    execute
    self
    IsActiveParticipant:bool=HasMatchingGameplayTag!(
     self:Gameplay Tag Asset Interface Interface=GetAbilitySystemComponent!(Actor:Actor Object Reference=@n12) -> (ReturnValue:Ability System Component Object Reference)
     TagToCheck:Gameplay Tag Structure=Get ActiveVaultParticipantTag()
    ) -> (ReturnValue:bool) {is=title K2Node_CallFunction}
   ) -> (then)
   Branch(execute=@n16.Removed, Condition=HasMatchingGameplayTag(self=GetAbilitySystemComponent(Actor=@n12), TagToCheck=Get VaultParticipantTag()))
   > SetIsActiveParticipant(IsActiveParticipant=false)
   SetIsActiveParticipant(execute=@n15.Added, IsActiveParticipant=true)
   n17 = Branch(execute=@n14.Removed, Condition=HasMatchingGameplayTag(self=GetAbilitySystemComponent(Actor=@n12), TagToCheck=Get VaultParticipantTag()))
   PlayEnterExit(execute=@n17.else, IsEntering=false)
   PlayEnterExit(execute=@n13.Added, IsEntering=true)
   ListenToWeakpointDamage() {is=title K2Node_CustomEvent}
   > n18 = StartListeningToEvent(
    Target=Get VaultEventRouter()
    EventType=/VividRazorCoreGameplay/Gameplay/Structs/Struct_VividRazorCore_Vault_Weapoint_Damaged_Payload
    None=Self-Reference()
   ) -> (Payload:Struct Vivid Razor Core Vault Weapoint Damaged Payload Structure)
   > Set WeakpointDamagedAsyncListener(WeakpointDamagedAsyncListener=@n18)
   OnWeakpointDamageEvent(execute=@n18.OnEventReceived)
   StopListeningToAlertStatusReplicated() {is=title K2Node_CustomEvent}
   > n19 = Get!(execute) -> (then, else, AlertStatusReplicatedAsyncListener:Async Action Start Listening To Event Object Reference)
   > StopListeningToEvent(self=@n19)
   > Set AlertStatusReplicatedAsyncListener()
   Event PreConstruct()
   > SetBrushFromMaterial(
    self=Get Image_ProgressBar_Latest() -> (Image_ProgressBar_Latest:Image Object Reference)
    Material=/VividRazorVaultGameplay/UI/Materials/M_UI_VividRazor_Frame
   )
   > SetWeakpointCount(Count=GetWeakpointCount() -> (Count=0))
   > SetMobileStyles()
   Event Construct()
   > n20 = Cast To Pawn!(execute, Object=GetViewTarget!(self:Controller Object Reference=GetOwningPlayer()) -> (ReturnValue:Actor Object Reference)) -> (then, CastFailed, AsPawn:Pawn Object Reference) {is=title K2Node_DynamicCast}
   > ListenForParticipant!(execute, self, Pawn:Pawn Object Reference=@n20) -> (then) {is=title K2Node_CallFunction}
   > InitializeFromRouter()
   > ListenForProgressUpdates()
   Event Destruct()
   > CleanupTimersAndReferences() {is=title K2Node_CallFunction}
   > StopListeningToAlertStatusReplicated() {is=title K2Node_CallFunction}
   > Cancel(self:Cancellable Async Action Object Reference=[
     Get ParticipantTagAddListener() -> (ParticipantTagAddListener:Ability Async Wait Gameplay Tag Added Object Reference)
     Get ParticipantTagRemoveListener() -> (ParticipantTagRemoveListener:Ability Async Wait Gameplay Tag Removed Object Reference)
     Get ActiveParticipantTagAddListener() -> (ActiveParticipantTagAddListener:Ability Async Wait Gameplay Tag Added Object Reference)
     Get ActiveParticipantTagRemoveListener() -> (ActiveParticipantTagRemoveListener:Ability Async Wait Gameplay Tag Removed Object Reference)
    ])
  ]}
 GetClosestManager {nodes [
   GetClosestManager() {is=title K2Node_FunctionEntry}
   > Set ViewTarget(ViewTarget:Actor Object Reference=GetViewTarget(self=GetOwningPlayer())) -> (Output_Get:Actor Object Reference)
   > n1 = GetAllActorsOfClass!(
    execute
    ActorClass:Actor Class Reference=/VividRazorVaultGameplay/Gameplay/Environmental/Manager/BGA_VividRazorVault_Manager.BGA_VividRazorVault_Manager_C
   ) -> (then, OutActors:Array of BGA Vivid Razor Vault Manager Object References)
   > n2 = For Each Loop(Array:Array of BGA Vivid Razor Vault Manager Object References=@n1) -> (Array Element:BGA Vivid Razor Vault Manager Object Reference)
   Return Node!(
    execute=@n2.Completed
    VaultManager:BGA Vivid Razor Vault Manager Object Reference=Get ClosestActiveVault() -> (ClosestActiveVault:BGA Vivid Razor Vault Manager Object Reference)
   )
   > n3 = Set CalculatedDistanceForIteration(CalculatedDistanceForIteration:float=GetSquaredDistanceTo!(self:Actor Object Reference=@n2, OtherActor:Actor Object Reference=Get ViewTarget() -> (ViewTarget:Actor Object Reference)) -> (ReturnValue:float)) -> (Output_Get:float)
   n4 = Get ShortestDistanceSqr() -> (ShortestDistanceSqr:double)
   > Branch(Condition=OR Boolean(A="float < float"(A=@n3, B=@n4), B="Nearly Equal (Float)"(A=@n4, B=0, ErrorTolerance=0)))
   > Set ShortestDistanceSqr(ShortestDistanceSqr:double=Get CalculatedDistanceForIteration() -> (CalculatedDistanceForIteration:float)) -> (Output_Get:double)
   > Set ClosestActiveVault(ClosestActiveVault:BGA Vivid Razor Vault Manager Object Reference=@n2) -> (Output_Get:BGA Vivid Razor Vault Manager Object Reference)
  ]}
 GetDesiredNextWeakpoint {nodes [
   GetDesiredNextWeakpoint() {is=title K2Node_FunctionEntry}
   n1 = Get DynamicEntryBox_Weakpoints() -> (DynamicEntryBox_Weakpoints:Dynamic Entry Box Object Reference)
   n2 = GetWeakpointCount() -> (Count=0)
   n3 = float / float(A=GetProgress!(self) -> (Progress:double), B=float / float(A=1, B="To Float (Integer)"(InInt=@n2)))
   n4 = Truncate(A=@n3)
   n5 = Round(A=@n3) {is=title K2Node_CallFunction}
   n6 = "float > float"(A="% (Float)"!(A:double=@n3, B:double=1) -> (ReturnValue:double), B=0.999)
   > Return Node!(
    execute
    DesiredNextWeakpoint:int="Min (Integer)"(
     A=Select(Option 0:int=@n4, Option 1:int=@n5, Index=@n6) -> (ReturnValue:int)
     B=GetNumEntries!(self:Dynamic Entry Box Base Object Reference=@n1) -> (ReturnValue:int=0)
    ) -> (ReturnValue=0)
   )
  ]}
 GetProgress {nodes [
   GetProgress() {is=title K2Node_FunctionEntry}
   n1 = Get CurrentManager() -> (CurrentManager:BGA Vivid Razor Vault Manager Object Reference)
   > n2 = Is Valid(InputObject=@n1)
   Return Node!(execute=@n2.Is Not Valid, Progress:double=0)
   > Return Node(Progress=Get Progress!(self:BGA Vivid Razor Vault Manager Object Reference=@n1) -> (Progress:double))
  ]}
 GetWeakpointCount {nodes [
   GetWeakpointCount() {is=title K2Node_FunctionEntry}
   n1 = Get CurrentManager()
   > n2 = Is Valid(InputObject=@n1)
   Return Node!(execute=[@n4.Is Not Valid, @n2.Is Not Valid], Count:int=5)
   n3 = Get VaultActor!(self:BGA Vivid Razor Vault Manager Object Reference=@n1) -> (VaultActor:BGA Vivid Razor Vault Door Parent Object Reference)
   > n4 = Is Valid(InputObject=@n3)
   > Return Node(Count=Get WeakpointCount!(self:BGA Vivid Razor Vault Door Parent Object Reference=@n3) -> (WeakpointCount:int=0))
  ]}
 InitializeFromRouter {nodes [
   n1 = InitializeFromRouter!() -> (then, GameplayEventRouter:Gameplay Event Router Component Object Reference) {is=title K2Node_FunctionEntry}
   > Set bInitNewRouter(bInitNewRouter:bool="Not Equal (Object)"(A:Object Reference=@n1, B:Object Reference=Get VaultEventRouter()) -> (ReturnValue:bool=~) {is=title K2Node_PromotableOperator}) -> (Output_Get:bool)
   > ClearRouterBindings() {is=title K2Node_CallFunction}
   > n2 = Get!(execute) -> (then, else, GameplayEventRouter:Gameplay Event Router Component Object Reference)
   n3 = GetEventRouter!(
    execute=@n2.else
    Scope:Object Class Reference=/VividRazorVaultGameplay/Gameplay/Environmental/Manager/BP_VividRazorVaultPlayspace.BP_VividRazorVaultPlayspace_C
    ContextActor:Actor Object Reference=GetViewTarget(self=GetOwningPlayer())
   ) -> (then, ReturnValue:Gameplay Event Router Component Object Reference)
   n4 = Merge Execution with Data!(
    Option A:Exec=@n2.then
    Option A Data:Gameplay Event Router Component Object Reference=@n2
    Option B:Exec=@n3.then
    Option B Data:Gameplay Event Router Component Object Reference=@n3
   ) -> (Then:Exec, Data:Gameplay Event Router Component Object Reference) {is=title K2Node_MacroInstance}
   > Set VaultEventRouter(VaultEventRouter=@n4)
   > n5 = Cast To BP_VividRazorVaultPlayspace!(execute, Object=GetOwner(self=Get VaultEventRouter())) -> (then, CastFailed, AsBP Vivid Razor Vault Playspace:BP Vivid Razor Vault Playspace Object Reference) {is=title K2Node_DynamicCast}
   > n6 = Set CurrentManager(CurrentManager=Get VaultManager!(self:BP Vivid Razor Vault Playspace Object Reference=@n5) -> (VaultManager:BGA Vivid Razor Vault Manager Object Reference))
   > Is Valid(InputObject=@n6)
   n7 = Get Progress(self=@n6)
   > n8 = Branch(Condition=AND Boolean(A="float < float"(A=@n7, B=1), B=Get bAlert!(self:BGA Vivid Razor Vault Manager Object Reference=@n6) -> (bAlert:bool)))
   > SetWeakpointCountFromVault() {is=title K2Node_CallFunction}
   > ListenToWeakpointDestroyed() {is=title K2Node_CallFunction}
   > ListenToWeakpointDamage() {is=title K2Node_CallFunction}
   > ListenToEndEvent() {is=title K2Node_CallFunction}
   > ListenForProgressUpdates()
   > SetProgress!(execute, self, Progress:float=GetProgress() {is=title K2Node_CallFunction}) -> (then)
   > InitializeWeakpointStates!(execute, self, bForceUpdateWidgets:bool=Get bInitNewRouter() -> (bInitNewRouter:bool)) -> (then)
   > StopListeningToAlertStatusReplicated()
   CleanupTimersAndReferences(execute=@n8.else)
   > PlayEnterExit(IsEntering=false)
   > ListenToAlertStatusReplicated() {is=title K2Node_CallFunction}
  ]}
 InitializeWeakpointStates {nodes [
   InitializeWeakpointStates!() -> (then, bForceUpdateWidgets:bool=~) {is=title K2Node_FunctionEntry}
   > Set NextWeakpointIndex(NextWeakpointIndex:int=GetDesiredNextWeakpoint!(self) -> (DesiredNextWeakpoint:int=0) {is=title K2Node_CallFunction}) -> (Output_Get:int=0)
   > n1 = For Each Loop(Array:Array of User Widget Object References=GetAllEntries!(self:Dynamic Entry Box Base Object Reference=Get DynamicEntryBox_Weakpoints()) -> (ReturnValue:Array of User Widget Object References)) -> (Array Element:User Widget Object Reference)
   > n2 = Cast To WBP_VividRazorVault_ProgressBar_Weakpoint!(execute, Object=@n1) -> (then, CastFailed, AsWBP Vivid Razor Vault Progress Bar Weakpoint:WBP Vivid Razor Vault Progress Bar Weakpoint Object Reference) {is=title K2Node_DynamicCast}
   n3 = Get NextWeakpointIndex() -> (NextWeakpointIndex:int=0)
   > SetWeakpointDestroyed!(
    execute
    self:WBP Vivid Razor Vault Progress Bar Weakpoint Object Reference=@n2
    bForceUpdate:bool=Get bForceUpdateWidgets() -> (bForceUpdateWidgets:bool)
    IsDestroyed:bool=AND Boolean(A="integer < integer"(A=@n1.Array Index, B=@n3), B=Get IsActiveParticipant() -> (IsActiveParticipant:bool))
   ) -> (then)
  ]}
 ListenForProgressUpdates {nodes [
   ListenForProgressUpdates() {is=title K2Node_FunctionEntry}
   n1 = Get ProgressUpdateTimerHandle()
   > Branch(Condition=NOT Boolean(A=Is Timer Active by Handle!(Handle:Timer Handle Structure=@n1) -> (ReturnValue:bool)))
   > n2 = Set Timer by Event(Delegate=Create Event(), Time=0.5, bLooping=true)
   > Set ProgressUpdateTimerHandle(ProgressUpdateTimerHandle:Timer Handle Structure=@n2) -> (Output_Get:Timer Handle Structure)
  ]}
 OnWeakpointDamageEvent {nodes [
   OnWeakpointDamageEvent() {is=title K2Node_FunctionEntry}
   > n1 = Branch(Condition=Get IsWeakpointDamageActive() -> (IsWeakpointDamageActive:bool))
   n2 = Set Timer by Event(
    execute=[@n3.then, @n1.then]
    Delegate=Create Event()
    Time=Get DurationOfWeakpointDamageState() -> (DurationOfWeakpointDamageState:float)
    bLooping=false
   )
   > Set WeakpointDamageEventTimer(WeakpointDamageEventTimer:Timer Handle Structure=@n2) -> (Output_Get:Timer Handle Structure)
   Set IsWeakpointDamageActive(execute=@n1.else, IsWeakpointDamageActive:bool=true) -> (Output_Get:bool)
   > PlayAnimation(InAnimation=Get Anim_WeakpointDamageBegin() -> (Anim_WeakpointDamageBegin:Widget Animation Object Reference))
   > n3 = StopAnimation(InAnimation=Get Anim_WeakpointDamageEnd() -> (Anim_WeakpointDamageEnd:Widget Animation Object Reference))
  ]}
 OnWeakpointDamageEventExpired {nodes [
   OnWeakpointDamageEventExpired() {is=title K2Node_FunctionEntry}
   > Branch(Condition=Get IsWeakpointDamageActive())
   > Set IsWeakpointDamageActive(IsWeakpointDamageActive=false)
   > PlayAnimation(InAnimation=Get Anim_WeakpointDamageEnd())
   > StopAnimation(InAnimation=Get Anim_WeakpointDamageBegin())
  ]}
 OnWeakpointDestroyed {nodes [
   OnWeakpointDestroyed() {is=title K2Node_FunctionEntry}
   n1 = Get DynamicEntryBox_Weakpoints()
   > Branch(Condition="integer > integer"(A=GetNumEntries(self=@n1) -> (ReturnValue=0), B=Get NextWeakpointIndex() -> (NextWeakpointIndex=0)))
   n2 = Get DynamicEntryBox_Weakpoints()
   > n3 = Cast To WBP_VividRazorVault_ProgressBar_Weakpoint(Object="Get (a copy)"(Array:Array of User Widget Object References=GetAllEntries(self=@n2), Dimension 1=Get NextWeakpointIndex() -> (NextWeakpointIndex=0)) -> (Output:User Widget Object Reference))
   > n4 = PlayWeakpointDestroyed(self:WBP Vivid Razor Vault Progress Bar Weakpoint Object Reference=@n3)
   Set NextWeakpointIndex(execute=[@n4.then, @n3.CastFailed], NextWeakpointIndex=int + int(A=Get NextWeakpointIndex() -> (NextWeakpointIndex=0))) -> (Output_Get=0)
   > PlayAnimation(InAnimation=Get Anim_WeakpointDestroyed() -> (Anim_WeakpointDestroyed:Widget Animation Object Reference))
  ]}
 PlayEnterExit {nodes [
   n1 = PlayEnterExit!() -> (then, IsEntering:bool=~) {is=title K2Node_FunctionEntry}
   > n2 = Delta Gate Wildcard!(execute:Exec, Variable:"Boolean (by ref)"=Get IsVaultUIShowing() -> (IsVaultUIShowing:bool), New Value:bool=@n1) -> (Success:Exec, Updated Variable:"Boolean (by ref)", Previous Variable:bool=~) {is=title K2Node_MacroInstance}
   > n3 = Branch(Condition=@n2)
   n4 = Get Anim_EnterExit() -> (Anim_EnterExit:Widget Animation Object Reference)
   > PlayAnimationForward(InAnimation=@n4)
   > InitializeFromRouter() {is=title K2Node_CallFunction}
   PlayAnimationReverse(execute=@n3.else, InAnimation=@n4)
  ]}
 SetIsActiveParticipant {nodes [
   n1 = SetIsActiveParticipant!() -> (then, IsActiveParticipant:bool=~) {is=title K2Node_FunctionEntry}
   > n2 = Set IsActiveParticipant(IsActiveParticipant:bool=@n1) -> (Output_Get:bool)
   > n3 = Branch(Condition=@n2)
   > SetProgress(Progress=GetProgress())
   > InitializeWeakpointStates(bForceUpdateWidgets=false) {is=title K2Node_CallFunction}
   > SetVisibility(self=Get Text_Status() -> (Text_Status:Common Text Object Reference), InVisibility=Hidden)
   SetProgress(execute=@n3.else, Progress=0)
   > SetVisibility(self=Get Text_Status(), InVisibility=SelfHitTestInvisible)
  ]}
 SetMobileStyles {nodes [
   SetMobileStyles() {is=title K2Node_FunctionEntry}
   n1 = GetOwningPlayer()
   > Branch(Condition=AND Boolean(A=IsMobileGame(), B=IsUsingTouch(PlayerController=@n1)))
   n2 = Get Overlay_ProgressBar() -> (Overlay_ProgressBar:Overlay Object Reference)
   > SetRenderTransform!(
    execute
    self:Widget Object Reference=@n2
    InTransform_Translation:Vector 2D Structure=(X=0,Y=0)
    InTransform_Scale:Vector 2D Structure=(X=0.75,Y=0.75)
    InTransform_Shear:Vector 2D Structure=(X=0,Y=0)
    InTransform_Angle:float=0
   ) -> (then)
   > SetRenderTranslation(self=@n2, Translation=(X=0,Y=15))
  ]}
 SetProgress {nodes [
   n1 = SetProgress!() -> (then, Progress:float=~) {is=title K2Node_FunctionEntry}
   n2 = PrintString!(execute, InString="To String (Float)"(InDouble=@n1), bPrintToScreen, bPrintToLog, TextColor, Duration, Key) -> (then)
   n3 = Branch(execute=[@n2.then, @n1.then], Condition=Get IsActiveParticipant())
   > n4 = GetDynamicMaterial(self=Get Image_ProgressBar_Latest())
   > SetScalarParameterValue(self=@n4, ParameterName=Progress, Value=@n1)
   n5 = GetDynamicMaterial(execute=@n3.else, self=Get Image_ProgressBar_Latest())
   > SetScalarParameterValue(self=@n5, ParameterName=Progress, Value=0)
  ]}
 SetWeakpointCount {nodes [
   n1 = SetWeakpointCount!() -> (then, Count:int) {is=title K2Node_FunctionEntry}
   > n2 = GetDynamicMaterial(self=Get Image_ProgressBar_Latest())
   > SetScalarParameterValue(self=@n2, ParameterName=Weakpoint Count, Value="To Float (Integer)"(InInt=@n1))
   > n3 = GetDynamicMaterial(self=Get Image_ProgressBar_Latest())
   > n4 = Get Scalar Parameter Value!(execute, self:Material Instance Dynamic Object Reference=@n3, ParameterName:Name=ProgressStart) -> (then, ReturnValue:float) {is=title K2Node_CallFunction}
   > Set ProgressStart(ProgressStart:float=@n4) -> (Output_Get:float)
   > n5 = Get Scalar Parameter Value(self=@n3, ParameterName=ProgressEnd)
   > Set ProgressEnd(ProgressEnd:float=@n5) -> (Output_Get:float)
   n6 = Get Count() -> (Count:int=0)
   n7 = float - float(A=Get ProgressEnd() -> (ProgressEnd:float), B=Get ProgressStart() -> (ProgressStart:float))
   n8 = Get WidthOverride!(self:Size Box Object Reference=Get SizeBox_ProgressBar() -> (SizeBox_ProgressBar:Size Box Object Reference)) -> (WidthOverride:float)
   > Set SpacingBetweenWeakpoints(SpacingBetweenWeakpoints:double=float / float(A=float * float(A=@n8, B=@n7) {is=title K2Node_PromotableOperator}, B="To Float (Integer)"(InInt=@n6))) -> (Output_Get:double)
   n9 = Get SizeBox_ProgressBar()
   > Set StartOffset(StartOffset:double=float * float(A=Get WidthOverride(self=@n9), B=Get ProgressStart())) -> (Output_Get:double)
   > Reset!(execute, self:Dynamic Entry Box Object Reference=Get DynamicEntryBox_Weakpoints(), bDeleteWidgets:bool=false) -> (then)
   > n10 = For Loop(FirstIndex=1, LastIndex=int - int(A=Get Count() -> (Count=0)))
   > n11 = Create Entry!(execute, self:Dynamic Entry Box Object Reference=Get DynamicEntryBox_Weakpoints()) -> (then, ReturnValue:User Widget Object Reference)
   n12 = float * float(A="To Float (Integer)"(InInt=@n10), B=Get SpacingBetweenWeakpoints() -> (SpacingBetweenWeakpoints:double))
   n13 = Get StartOffset() -> (StartOffset:double)
   > SetRenderTranslation(self=@n11, Translation=MakeVector2D(X=float + float(A=@n13, B=@n12), Y=0))
  ]}
 UpdateProgress {nodes [
   UpdateProgress() {is=title K2Node_FunctionEntry}
   > n1 = Branch(Condition=NOT Boolean(A=IsSpatialGameplayExperience!() -> (ReturnValue:bool)))
   Branch(execute=[@n4.then, @n1.then, @n3.else], Condition=Get IsActiveParticipant())
   > SetProgress(Progress=GetProgress()) {is=title K2Node_CallFunction}
   n2 = GetEventRouter(
    execute=@n1.else
    Scope=&BP_VividRazorVaultPlayspace_C
    ContextActor=GetViewTarget(self=GetOwningPlayer())
   )
   > n3 = Branch(Condition="Not Equal (Object)"(A=@n2, B=Get VaultEventRouter()) -> (ReturnValue=~))
   > n4 = InitializeFromRouter(GameplayEventRouter=@n2)
  ]}
]
variables [
 ActiveParticipantTagAddListener {type=AbilityAsync_WaitGameplayTagAdded, default=null}
 ActiveParticipantTagRemoveListener {type=AbilityAsync_WaitGameplayTagRemoved, default=null}
 ActiveVaultParticipantTag {type=GameplayTag, default {tagName=Granted.VividRazorVault.Participant.Active}}
 AlertStatusReplicatedAsyncListener {type=AsyncAction_StartListeningToEvent, default=null}
 CurrentManager {type=BGA_VividRazorVault_Manager_C, default=null}
 Delta {type=number, default=0.1}
 DurationOfWeakpointDamageState {type=number, default=1}
 IsActiveParticipant {type=boolean}
 IsVaultUIShowing {type=boolean}
 IsWeakpointDamageActive {type=boolean}
 NextWeakpointIndex {type=integer}
 ParticipantTagAddListener {type=AbilityAsync_WaitGameplayTagAdded, default=null}
 ParticipantTagRemoveListener {type=AbilityAsync_WaitGameplayTagRemoved, default=null}
 ProgressUpdateTimerHandle {type=TimerHandle}
 Tween ID {type=integer}
 VaultEventRouter {type=GameplayEventRouterComponent, default=null}
 VaultOpenAsyncListener {type=AsyncAction_StartListeningToStatefulEvent, default=null}
 VaultParticipantTag {type=GameplayTag, default {tagName=Granted.VividRazorVault.Participant}}
 WeakpointDamagedAsyncListener {type=AsyncAction_StartListeningToEvent, default=null}
 WeakpointDamageEventTimer {type=TimerHandle}
 WeakpointDestroyedAsyncListener {type=AsyncAction_StartListeningToEvent, default=null}
]
widgets {
 info {inheritedWidgetCount=1, parentClass=&FortHUDElementWidget, rootWidgetClass=&Overlay, widgetCount=10}
 widgets [
  Overlay_ProgressBar {
   widgetClassPath=&Overlay
   bIsVariable=true
   widget=Overlay_ProgressBar
   properties {bOverride_Cursor=true, renderTransform {angle=0, scale=<1 1>, shear=<0 0>, translation=<0 35>}}
  }
  SizeBox_ProgressBar {
   widgetClassPath=&SizeBox
   bIsVariable=true
   parent=Overlay_ProgressBar
   widget=SizeBox_ProgressBar
   properties {bOverride_HeightOverride=true, bOverride_WidthOverride=true, heightOverride=85, widthOverride=800}
   slot {class=OverlaySlot, properties {horizontalAlignment=HAlign_Fill, verticalAlignment=VAlign_Center}}
  }
  Overlay_70 {widgetClassPath=&Overlay, parent=SizeBox_ProgressBar, widget=Overlay_70, slot {class=SizeBoxSlot}}
  Image_ProgressBar_Latest {
   widgetClassPath=&Image
   bIsVariable=true
   parent=Overlay_70
   widget=Image_ProgressBar_Latest
   properties {brush {
     bIsDynamicallyLoaded=false
     drawAs=Image
     imageType=FullColor
     mirroring=NoMirror
     resourceName=None
     resourceObject=&M_UI_VividRazor_Frame
     tiling=NoTile
     imageSize=<32 32>
     margin=<ltrb 0 0 0 0>
     outlineSettings {
      bUseBrushTransparency=false
      roundingType=HalfHeightRadius
      width=0
      color {colorUseRule=UseColor_Specified, specifiedColor=<rgba 0 0 0 0>}
      cornerRadii=<0 0 0 0>
     }
     tintColor {colorUseRule=UseColor_Specified, specifiedColor=<rgba 1 1 1 1>}
     uVRegion {bIsValid=false, max=<0 0>, min=<0 0>}
    }}
   slot {class=OverlaySlot, properties {horizontalAlignment=HAlign_Fill, verticalAlignment=VAlign_Fill}}
  }
  DynamicEntryBox_Weakpoints {
   widgetClassPath=&DynamicEntryBox
   bIsVariable=true
   parent=Overlay_70
   widget=DynamicEntryBox_Weakpoints
   properties {
    entryBoxType=Overlay
    entryWidgetClass=&/WBP_VividRazorVault_ProgressBar_Weakpoint.WBP_VividRazorVault_ProgressBar_Weakpoint_C
    numDesignerPreviewEntries=4
    renderTransform {angle=0, scale=<1 1>, shear=<0 0>, translation=<0 -56>}
   }
   slot {class=OverlaySlot, properties {verticalAlignment=VAlign_Center}}
  }
  Border_BoostGlow {
   widgetClassPath=&Border
   bIsVariable=true
   parent=Overlay_ProgressBar
   widget=Border_BoostGlow
   properties {
    renderOpacity=0
    visibility=Collapsed
    background {
     bIsDynamicallyLoaded=false
     drawAs=NoDrawType
     imageType=NoImage
     mirroring=NoMirror
     resourceName=None
     resourceObject=null
     tiling=NoTile
     imageSize=<32 32>
     margin=<ltrb 0 0 0 0>
     outlineSettings {
      bUseBrushTransparency=false
      roundingType=HalfHeightRadius
      width=0
      color {colorUseRule=UseColor_Specified, specifiedColor=<rgba 0 0 0 0>}
      cornerRadii=<0 0 0 0>
     }
     tintColor {colorUseRule=UseColor_Specified, specifiedColor=<rgba 1 1 1 1>}
     uVRegion {bIsValid=false, max=<0 0>, min=<0 0>}
    }
    desiredSizeScale=<0.5 1>
    padding=<ltrb 0 0 0 0>
    renderTransform {angle=0, scale=<1 0.35>, shear=<0 0>, translation=<0 0>}
    renderTransformPivot=<0 0.5>
   }
   slot {class=OverlaySlot, properties {horizontalAlignment=HAlign_Fill, verticalAlignment=VAlign_Fill}}
  }
  SizeBox {
   widgetClassPath=&SizeBox
   parent=Border_BoostGlow
   widget=SizeBox
   properties {bOverride_WidthOverride=true, widthOverride=800}
   slot {class=BorderSlot, properties {padding=<ltrb 0 0 0 0>}}
  }
  BoostGlow {
   widgetClassPath=&Image
   bIsVariable=true
   parent=SizeBox
   widget=BoostGlow
   properties {
    brush {
     bIsDynamicallyLoaded=false
     drawAs=Box
     imageType=FullColor
     mirroring=NoMirror
     resourceName=None
     resourceObject=/VividRazorVaultGameplay/UI/Materials/MI_UI_VividRazor_BoostGlow
     tiling=NoTile
     imageSize=<72 72>
     margin=<ltrb 0.5 0.5 0.5 0.5>
     outlineSettings {
      bUseBrushTransparency=false
      roundingType=HalfHeightRadius
      width=0
      color {colorUseRule=UseColor_Specified, specifiedColor=<rgba 0 0 0 0>}
      cornerRadii=<0 0 0 0>
     }
     tintColor {colorUseRule=UseColor_Specified, specifiedColor=<rgba 1 1 1 1>}
     uVRegion {bIsValid=false, max=<0 0>, min=<0 0>}
    }
    renderTransformPivot=<0 0.5>
   }
   slot {class=SizeBoxSlot, properties {padding=<ltrb -32 -32 -32 -32>}}
  }
  Text_Status {
   widgetClassPath=&CommonTextBlock
   bIsVariable=true
   parent=Overlay_ProgressBar
   widget=Text_Status
   properties {
    mobileFontSizeMultiplier=0.8
    text="Robbery in-progress!"
    textTransformPolicy=ToUpper
    colorAndOpacity {colorUseRule=UseColor_Specified, specifiedColor=<rgba 1 0.7215 0.0002 1>}
    font {
     bForceMonospaced=false
     bMaterialIsStencil=false
     fontMaterial=null
     fontObject=&BurbankSmall
     letterSpacing=0
     monospacedWidth=1
     size=24
     skewAmount=0
     typefaceFontName=Black
     outlineSettings {
      bApplyOutlineToDropShadows=false
      bMiteredCorners=false
      bSeparateFillAlpha=false
      outlineMaterial=null
      outlineSize=2
      outlineColor=<rgba 0 0 0 1>
     }
    }
    renderTransform {angle=0, scale=<1 1>, shear=<0 0>, translation=<0 25>}
    shadowOffset=<0 0>
   }
   slot {class=OverlaySlot, properties {horizontalAlignment=HAlign_Center, verticalAlignment=VAlign_Bottom}}
  }
  TutorialHighlight {widgetClassPath=&TutorialHighlight}
 ]
}
